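# Physics asset for the FallValleyCharge backpack cosmetic, providing collision bodies and joints for its dangling bow and ribbon.
# The pack body combines a capsule and a box, while each bow and ribbon bone carries a small capsule.
# Three ribbon segments chain from the pack, and the top and bottom bow pieces attach directly to it.
# Swing and twist limits bound each joint, with the bow pieces allowed the widest range.
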PhysicsAsset /BRCosmetics/Accessories/FORT_Backpacks/Backpack_F_MED_FallValleyCharge/Meshes/F_MED_FallValleyCharge_Pack_Physics
bodies {
 dyn_bow_bottom [{shapeType=Capsule, length=1, radius=1.5, shapeName=dyn_bow_bottom_capsule, center=<3.5836 -0.6806 0.1286>, rotation=<rot -86 178.9998 -179>}]
 dyn_bow_top [{
   shapeType=Capsule
   length=0.7574
   radius=1.6162
   shapeName=dyn_bow_top_capsule
   center=<1.7047 1.0866 2.3405>
   rotation=<rot -33.0869 2.2713 -0.0328>
  }]
 dyn_ribbon_01 [{shapeType=Capsule, length=3.3982, radius=0.7348, shapeName=dyn_ribbon_01_capsule, center=<3.1927 0 0.0573>, rotation=<rot 87 179 179>}]
 dyn_ribbon_02 [{shapeType=Capsule, length=4.0353, radius=0.9647, shapeName=dyn_ribbon_02_capsule, center=<2.5 0 0>, rotation=<rot 83 -179 -179>}]
 dyn_ribbon_03 [{shapeType=Capsule, length=3.883, radius=1.117, shapeName=dyn_ribbon_03_capsule, center=<2.2998 -1.2365 -0.6757>, rotation=<rot 69 0 19.389>}]
 pack [
  {shapeType=Capsule, length=14.482, radius=15.518, shapeName=pack_capsule, center=<9.4894 14.3921 -2.4188>, rotation=<rot -2.0064 0 -0.0002>}
  {shapeType=Box, extentX=29.0203, extentY=6.5, extentZ=20, shapeName=pack_box, center=<-1.2462 -2.9874 0.6911>, rotation=<rot -15 0 0>}
 ]
}
constraints [
 {bone1Name=dyn_ribbon_03, bone2Name=dyn_ribbon_02, swing1LimitDegrees=65, swing2LimitDegrees=65, twistLimitDegrees=15}
 {bone1Name=dyn_ribbon_02, bone2Name=dyn_ribbon_01, swing1LimitDegrees=45, swing2LimitDegrees=45, twistLimitDegrees=15}
 {bone1Name=dyn_ribbon_01, bone2Name=pack, swing1LimitDegrees=55, swing2LimitDegrees=55, twistLimitDegrees=5}
 {bone1Name=dyn_bow_top, bone2Name=pack, swing1LimitDegrees=90, swing2LimitDegrees=90, twistLimitDegrees=20}
 {bone1Name=dyn_bow_bottom, bone2Name=pack, swing1LimitDegrees=90, swing2LimitDegrees=90, twistLimitDegrees=20}
]
