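# Animation Blueprint for the FallValleyCharge backpack cosmetic, adding secondary motion to its ribbon and bow bones.
# bRigidBodyEnabled picks between RigidBody simulation and a fallback of Trail, Spring and additive bone tweaks driven by pawn speed and jump/fall.
# Each update feeds the simulation scaled gravity and an eased backward force while jumping or falling outside skydiving.
# Additive root, ribbon and pack offsets follow the blend.
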AnimBlueprint /BRCosmetics/Accessories/FORT_Backpacks/Backpack_F_MED_FallValleyCharge/Meshes/F_MED_FallValleyCharge_Pack_AnimBP
target_skeleton=&/F_MED_FallValleyCharge_Pack_Skeleton
graphs [
 AnimGraph {nodes [
   n1 = Trail controller - Bone: dyn_ribbon_03(ComponentPose=Local To Component(LocalPose=Use cached pose 'SavedPose20702'()), Alpha=0.9) {
    bLimitRotation=true
    bUsePlanarLimit=true
    chainLength=3
    baseJoint {boneName=pack}
    fakeVelocity=<1 1 1>
    planarLimits [{drivingBone {boneName=root}, planeTransform {location=<0 -11 0>, rotation=<rot 0 15 -92>, scale=<1 1 1>}}]
    rotationLimits [{limitMax=<180 180 180>}, {limitMax=<180 180 180>}, {limitMax=<180 180 180>}]
    rotationOffsets [<0 0 -60>, <0 0 -60>, <0 0 -60>]
    trailBone {boneName=dyn_ribbon_03}
    trailRelaxationSpeed {editorCurveData {keys [<at 0 8 RCIM_Linear>,<at 1.3887 3.1994 RCIM_Linear>,<at 3.5441 1 RCIM_Linear>]}}
   }
   n2 = "Transform (Modify) Bone - Bone: dyn_ribbon_01"(ComponentPose=@n1, Alpha=Get Ribon_fix_onSprint() -> (Ribon_fix_onSprint:double), Translation="0,0,0", Rotation="0,0,-30", Scale="1,1,1") {
    rotationMode=BMM_Additive
    rotationSpace=BCS_ParentBoneSpace
    alphaBoolBlend {blendInTime=0.8, blendOption=QuarticInOut, blendOutTime=0.6}
    boneToModify {boneName=dyn_ribbon_01}
    rotation {roll=-30}
   }
   n3 = Get Jump_Fall() -> (Jump_Fall:bool)
   n4 = "Transform (Modify) Bone - Bone: dyn_bow_top"!(ComponentPose=@n2, bAlphaBoolEnabled:bool=@n3, Translation="0,0,0", Rotation="60,0,0", Scale="1,1,1") -> (Pose) {
    alphaInputType=Bool
    rotationMode=BMM_Additive
    rotationSpace=BCS_ParentBoneSpace
    alphaBoolBlend {blendInTime=0.8, blendOption=QuarticInOut, blendOutTime=0.6}
    boneToModify {boneName=dyn_bow_top}
    rotation {pitch=60}
   }
   n5 = "Transform (Modify) Bone - Bone: dyn_bow_bottom"!(ComponentPose=@n4, bAlphaBoolEnabled:bool=@n3, Translation="0,0,0", Rotation="60,0,0", Scale="1,1,1") -> (Pose) {
    alphaInputType=Bool
    rotationMode=BMM_Additive
    rotationSpace=BCS_ParentBoneSpace
    alphaBoolBlend {blendInTime=0.8, blendOption=QuarticInOut, blendOutTime=0.6}
    boneToModify {boneName=dyn_bow_bottom}
    rotation {pitch=60}
   }
   n6 = Spring controller - Bone: dyn_bow_top!(ComponentPose:Component Space Pose Link Structure=@n5, Alpha:float=1) -> (Pose:Component Space Pose Link Structure) {
    is=title AnimGraphNode_SpringBone
    bLimitDisplacement=true
    bRotateX=true
    bRotateY=true
    bTranslateZ=false
    maxDisplacement=1
    springDamping=700
    alphaBoolBlend {blendOption=Linear}
    springBone {boneName=dyn_bow_top}
   }
   n7 = Spring controller - Bone: dyn_bow_bottom!(ComponentPose:Component Space Pose Link Structure=@n6, Alpha:float=1) -> (Pose:Component Space Pose Link Structure) {
    is=title AnimGraphNode_SpringBone
    bLimitDisplacement=true
    bRotateX=true
    bRotateY=true
    bTranslateZ=false
    maxDisplacement=1
    springDamping=700
    alphaBoolBlend {blendOption=Linear}
    springBone {boneName=dyn_bow_bottom}
   }
   n8 = Component To Local(ComponentPose=@n7)
   n9 = RigidBody!(
    ComponentPose=Local To Component(LocalPose=Use cached pose 'SavedPose20702'())
    Alpha
    OverrideWorldGravity:Vector=Get Gravity Override()
    ExternalForce=Get External Force() -> (External Force:Vector)
   ) -> (Pose) {
    bClampLinearTranslationLimitToRefPose=true
    bOverrideWorldGravity=true
    overridePhysicsAsset=&/F_MED_FallValleyCharge_Pack_Physics
    simulationSpace=BaseBoneSpace
    alphaBoolBlend {blendOption=Linear}
    baseBoneRef {boneName=pack}
    componentAppliedLinearAccClamp=<3000 3000 3000>
    componentLinearAccScale=<0.3 0.3 0.5>
    componentLinearVelScale=<0.3 0.3 0.5>
    overrideWorldGravity {z=-900}
    simSpaceSettings {
     maxAngularAcceleration=1000
     maxAngularVelocity=3
     maxLinearAcceleration=700
     maxLinearVelocity=500
     velocityScaleZ=0
     worldAlpha=1
     externalLinearDragV=<1 1>
    }
   }
   n10 = Component To Local(ComponentPose=@n9)
   n11 = Get bRigidBodyEnabled()
   n12 = "Transform (Modify) Bone - Bone: root"(
    ComponentPose=Local To Component(LocalPose=Blend Poses by bool(BlendPose_0=@n10, BlendPose_1=@n8, bActiveValue=@n11) {blendType=Linear})
    Translation="0,0,-4"
    Rotation="0,0,0"
   ) {
    translationMode=BMM_Additive
    translationSpace=BCS_BoneSpace
    alphaBoolBlend {blendOption=Linear}
    boneToModify {boneName=root}
    translation {z=-4}
   }
   n13 = "Transform (Modify) Bone - Bone: dyn_ribbon_03"!(ComponentPose=@n12, bAlphaBoolEnabled:bool=True, Translation="0,0,0", Rotation="0,-15,0", Scale="1,1,1") -> (Pose) {
    alphaInputType=Bool
    rotationMode=BMM_Additive
    rotationSpace=BCS_ParentBoneSpace
    alphaBoolBlend {blendInTime=0.3, blendOption=QuarticInOut, blendOutTime=0.6}
    boneToModify {boneName=dyn_ribbon_03}
    rotation {yaw=-15}
   }
   n14 = "Transform (Modify) Bone - Bone: pack"!(ComponentPose=@n13, bAlphaBoolEnabled:bool, Translation="0,-1.5,-1", Rotation="-10,0,0", Scale) -> (Pose) {
    alphaInputType=Bool
    rotationMode=BMM_Additive
    rotationSpace=BCS_ParentBoneSpace
    translationMode=BMM_Additive
    translationSpace=BCS_ParentBoneSpace
    alphaBoolBlend {blendInTime=0.3, blendOption=QuarticInOut, blendOutTime=0.5}
    boneToModify {boneName=pack}
    rotation {pitch=-10}
    translation {y=-1.5, z=-1}
   }
   Output Pose(Result=Component To Local(ComponentPose=@n14)) {layerGroup=None}
   Save cached pose 'SavedPose20702'(Pose=Input Pose()) {cacheName=SavedPose20702}
  ]}
 EventGraph {nodes [
   n1 = Event BlueprintUpdateAnimation()
   > n2 = Gravity Override State Function DEPRECATED(CharacterPartAnimInstance=Self-Reference(), GameTimeInSeconds=GetGameTimeInSeconds(), GravityOverrideParams=Get Gravity Override Params())
   > Set Gravity Override(Gravity Override=vector / vector(A:Vector=@n2, B=1.4) -> (ReturnValue:Vector))
   n3 = AND Boolean(A=OR Boolean(A=Get bIsFalling(), B=Get bIsJumping()), B=NOT Boolean(A=Get bIsSkyDiving()))
   > n4 = Set Current Force(Current Force:double=FInterpTo(
     Current=Get Current Force() -> (Current Force:double)
     Target=SelectFloat(A=-30000, B=0, bPickA=@n3)
     DeltaTime=@n1.DeltaTimeX
     InterpSpeed=15
    )) -> (Output_Get:double)
   n5 = TryGetPawnOwner()
   > Set External Force(External Force:Vector=vector * vector(A=GetActorForwardVector!(self:Actor Object Reference=@n5) -> (ReturnValue:Vector), B=@n4) -> (ReturnValue=~)) -> (Output_Get:Vector)
   > Set Jump_Fall(Jump_Fall:bool=OR Boolean(A=Get bIsJumping(), B=Get bIsFalling())) -> (Output_Get:bool)
   n6 = Get PawnVelocity()
   > Set Ribon_fix_onSprint(Ribon_fix_onSprint:double=MapRangeClamped(Value=Vector Length(A=@n6), InRangeA=0, InRangeB=550)) -> (Output_Get:double)
   TryGetPawnOwner()
  ]}
]
variables [
 Alpha {type=number, default=1}
 Character Part Anim Instance {type=CustomCharacterPartAnimInstance, default=null}
 Current Force {type=number}
 External Force {type=Vector}
 Gravity Override {type=Vector}
 Gravity Override Params {type=GravityOverrideParamsStruct}
 Jump_Fall {type=boolean}
 Ribon_fix_onSprint {type=number}
]
